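# Niagara system for the Strobe drift trail vehicle cosmetic, with ribbon trails and wheel-wrap meshes.
# Ribbons002 spawns short-lived, backward-moving ribbons on the vehicle skeleton, fanning one particle set into eleven ribbon renderers through per-layer position, color and material bindings for strobes and bloom.
# WheelWrap001 pins meshes to wheel bones and scales them per wheel.
# Both emitters support painted color variants and per-tire disabling.
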NiagaraSystem /VehicleCosmetics/DriftTrail/DriftTrail_Strobe/FX/NS_Trail_Strobe_MP
dependencies {
 usedDataInterfaces [&NiagaraDataInterfaceCurve, &NiagaraDataInterfaceVectorCurve, &NiagaraDataInterfaceSkeletalMesh]
 usedDynamicInputs [&RandomRangeFloat]
 usedModules [
  &self:SystemSpawnScript.SystemScriptSource.SystemScriptGraph.NiagaraNodeAssignment_0.SetVariables_#1
  &SystemState
  &FN_UnpackVehicleFlags
  /Game/Effects/Niagara/Modules/Vehicle/FN_UpdateVehicleAttribs
  /Game/Effects/Niagara/Modules/Vehicle/FN_VehicleFlags_Cheatcodes
 ]
}
emitters {
 Ribbons002 {
  emitterUpdateScript [
   &EmitterState(Inactive Response=::NewEnumerator1, Life Cycle Mode=::NewEnumerator1, Loop Duration=1, UseLoopDelay=false)
   &VC_RateQualityHandler(Epic Spawn Rate=90, High Spawn Rate=40, Low Spawn Rate=30, Medium Spawn Rate=35, Use Custom Rate?=true) {enabled=false}
   &VC_SpawnRibbon_v1(Number of Ribbons=2, Spawn Time=-1, SpawnRate=120)
   /SetVariables(
    Emitter.Mask_Dots001=2.25
    Emitter.Mask_Dots002=-0.5
    Emitter.Mask_Dots_1:NiagaraFloat=&Sine(
     Normalized Angle=$Emitter.Age
     Period=1
     Scale=&FloatFromCurve(FloatCurve=DataInterface() {propertyValues {Curve {keys [<at 0 1>,<at 0.5 2>,<at 1 1>]}}}, CurveIndex=$Emitter.NormalizedLoopAge)
     Bias=2
    )
    Emitter.Velocity_Mask:NiagaraFloat=&FloatFromCurve(
     FloatCurve=DataInterface() {propertyValues {Curve {keys [
         <at 0 1>
         {interpMode=RCIM_Cubic, tangentMode=RCTM_Break, time=0.4, value=1}
         {interpMode=RCIM_Cubic, tangentMode=RCTM_Break, time=0.5, value=-1}
         {interpMode=RCIM_Cubic, tangentMode=RCTM_Break, time=0.9, value=-1}
         <at 1 1>
        ]}}}
     CurveIndex=$Engine.System.Age
     Scale Curve=-0.5
    )
    Emitter.Wave_2_Ribbon:Vector3f=&VectorFromCurve(
     VectorCurve=DataInterface() {propertyValues {
       XCurve {keys [
         {interpMode=RCIM_Cubic, tangentMode=RCTM_Break, time=0, value=2}
         {interpMode=RCIM_Cubic, tangentMode=RCTM_Break, time=0.5, value=1}
         {interpMode=RCIM_Cubic, tangentMode=RCTM_Break, time=1, value=2}
        ]}
       YCurve {keys [<at 0 0>,<at 1 0>]}
       ZCurve {keys [<at 0 0>,<at 1 0>]}
      }}
     CurveIndex=$Emitter.NormalizedLoopAge
    )
    Emitter.Wave_ribbon:Vector3f=&VectorFromCurve(
     VectorCurve=DataInterface() {propertyValues {
       XCurve {keys [
         {interpMode=RCIM_Cubic, tangentMode=RCTM_Break, time=0, value=1}
         {interpMode=RCIM_Cubic, tangentMode=RCTM_Break, time=0.5, value=2}
         {interpMode=RCIM_Cubic, tangentMode=RCTM_Break, time=1, value=1}
        ]}
       YCurve {keys [<at 0 0>,<at 1 0>]}
       ZCurve {keys [<at 0 0>,<at 1 0>]}
      }}
     CurveIndex=$Emitter.NormalizedLoopAge
    )
   ) {isSetParameters=true}
  ]
  particleSpawnScript [
   &VC_PaintedVariantsOverride(
    Black=<rgba 0.04 0.04 0.04 0.35>
    BurntSienna=<rgba 0.25 0.1262 0.0375 0.15>
    Cobalt=<rgba 0.0604 0.1667 0.5521 0.25>
    Crimson=<rgba 0.4531 0.0769 0.1334 0.2>
    ForestGreen=<rgba 0.0444 0.35 0 0.2>
    Gold:LinearColor=<rgba 0.4479 0.3622 0.105 0.2>
    Grey=<rgba 0.4479 0.3622 0.105 0.2>
    Lime=<rgba 0.1257 0.4167 0.0933 0.3>
    Orange=<rgba 0.5156 0.1777 0.1101 0.25>
    Pink=<rgba 0.5 0.0964 0.4194 0.2>
    Purple=<rgba 0.2471 0.0564 0.474 0.25>
    R1 Black:LinearColor=<rgba 0.151 0.151 0.151 0.35>
    R1 BurntSienna:LinearColor=<rgba 0.6042 0.1589 0.0071 0.2>
    R1 Cobalt:LinearColor=<rgba 0.146 0.1684 0.6094 0.25>
    R1 Crimson:LinearColor=<rgba 0.4896 0 0 0.2>
    R1 ForestGreen:LinearColor=<rgba 0.0659 0.4323 0 0.25>
    R1 Gold:LinearColor=<rgba 0.5052 0.4184 0.1579 0.3>
    R1 Grey:LinearColor=<rgba 0.5052 0.4184 0.1579 0.3>
    R1 Lime:LinearColor=<rgba 0.3213 0.4323 0 0.25>
    R1 Orange:LinearColor=<rgba 0.5156 0.3072 0.0903 0.3>
    R1 Pink:LinearColor=<rgba 0.7859 0.386 0.7969 0.2>
    R1 Purple:LinearColor=<rgba 0.5284 0 0.5573 0.3>
    R1 Saffron:LinearColor=<rgba 0.6 0.5117 0 0.3>
    R1 SkyBlue:LinearColor=<rgba 0.059 0.6544 0.6667 0.3>
    R1 TitaniumWhite:LinearColor=<rgba 0.6406 0.6406 0.6406 0.3>
    R2 Black:LinearColor=<rgba 0.526 0.5092 0.4192 0.3>
    R2 BurntSienna:LinearColor=<rgba 0.474 0.474 0.474 0.4>
    R2 Cobalt:LinearColor=<rgba 0.4984 0.5308 1.1 0.4>
    R2 Crimson:LinearColor=<rgba 0.3906 0.2696 0.2625 0.2>
    R2 ForestGreen:LinearColor=<rgba 0.6982 0.8281 0.6901 0.2>
    R2 Gold:LinearColor=<rgba 0.6458 0.6458 0.6458 0.35>
    R2 Grey:LinearColor=<rgba 0.6458 0.6458 0.6458 0.35>
    R2 Lime:LinearColor=<rgba 0.6758 0.7188 0.5636 0.25>
    R2 Orange:LinearColor=<rgba 0.7708 0.5987 0.5259 0.35>
    R2 Pink:LinearColor=<rgba 0.7188 0.5615 0.6616 0.4>
    R2 Purple:LinearColor=<rgba 0.6604 0.569 0.7917 0.4>
    R2 Saffron:LinearColor=<rgba 0.8229 0.8019 0.7329 0.25>
    R2 SkyBlue:LinearColor=<rgba 0.6009 0.7989 0.8125 0.3>
    R2 TitaniumWhite:LinearColor=<rgba 0.7656 0.7656 0.7656 0.45>
    Saffron=<rgba 0.4 0.378 0.0073 0.25>
    SkyBlue=<rgba 0.0022 0.4213 0.43 0.25>
    TitaniumWhite=<rgba 0.7 0.7 0.7 0.2>
    Use R1 Painted Overrides=true
    Use R2 Painted Overrides=true
   )
   &VC_InitializeRibbon_v1(
    Color=&MakeCustomLinearColorFromBool(Bool=$User.bIsPainted, True Color=$Particles.Color, False Color=<rgba 0.1554 0.1875 0.8177 0.2>)
    Custom Offset=true
    Enable Skeletal Position=true
    Lifetime=0.4
    Local Vector Offset=&AddVectorToPosition(Position=<-25 0 0>, Vector=$User.offset)
    Offset Amount=1
    Ribbon Width=8
    Write Material Random=false
   )
   &AddVelocity(Velocity=&VC_Mask_Vectors(
     FrontEnd Value:Vector3f=<-400 0 0>
     blsFrontEndPreview:NiagaraBool=false
     SpeedMin Value:Vector3f=<-400 0 0>
     SpeedMax Value:Vector3f=<-1 0 0>
     Speed Alpha:NiagaraFloat=$System.Speed_Alpha
    )) {enabled=false, module=AddVelocity001}
   &Color(Color=&MakeCustomLinearColorFromBool(Bool=$User.bIsPainted, True Color=$Particles.Color, False Color=<rgba 0.0447 0.0517 0.4 0.15>)) {enabled=false}
   &AddVelocity(Rotation Mode=::NewEnumerator2, Velocity=<-2500 0 0>)
   /VehicleCosmetics/FX/Niagara/Modules/MergeEmitters/VC_ExtraBindingColors(
    ExtraColors=2
    R1 Color:LinearColor=&MakeCustomLinearColorFromBool(Bool=$User.bIsPainted, True Color=$Particles.R1 Color, False Color=<rgba 0.4896 0.2012 0.0688 0.2>)
    R2 Color:LinearColor=&MakeCustomLinearColorFromBool(Bool=$User.bIsPainted, True Color=$Particles.R2 Color, False Color=<rgba 0.3393 0.4436 1 0.9826>)
   )
  ]
  particleUpdateScript [
   &ScaleColor(
    Scale Alpha=&FloatFromCurve(FloatCurve=DataInterface() {propertyValues {Curve {keys [
         {interpMode=RCIM_Cubic, tangentMode=RCTM_User, time=0, value=0}
         {interpMode=RCIM_Cubic, tangentMode=RCTM_User, time=0.1, value=2}
         {interpMode=RCIM_Cubic, tangentMode=RCTM_User, time=1, value=0}
        ]}}})
    ScaleA=true
    ScaleRGB=true
   ) {enabled=false}
   &ScaleColor(
    Scale Alpha=&FloatFromCurve(
     FloatCurve=DataInterface() {propertyValues {Curve {keys [
         {arriveTangent=8.5748, interpMode=RCIM_Cubic, leaveTangent=8.5748, tangentMode=RCTM_User, time=0, value=2.25}
         {arriveTangent=1.9737, interpMode=RCIM_Cubic, leaveTangent=1.9737, tangentMode=RCTM_User, time=0.1, value=3}
         {arriveTangent=-7.3322, interpMode=RCIM_Cubic, leaveTangent=-7.3322, tangentMode=RCTM_User, time=0.7, value=1}
         {arriveTangent=-4.5533, interpMode=RCIM_Cubic, leaveTangent=-4.5533, tangentMode=RCTM_User, time=1, value=-1.5}
        ]}}}
     Scale Curve=$User.Color_ScaleCurve
    )
    ScaleA=true
    ScaleRGB=true
   ) {module=ScaleColor001}
   &FN_DisableTireFX(Mask Index Attribute:NiagaraInt32=$Particles.SpawnGroup, Mask Index FL=2, Mask Index FR=3, Mask Index RL=0, Mask Index RR=1)
   &InheritSourceMovement(
    Applied Position Scale=&Multiply_VectorByFloat(
     Vector=<1 1 1>
     Float=&FloatFromCurve(FloatCurve=DataInterface() {propertyValues {Curve {keys [
          {interpMode=RCIM_Cubic, leaveTangent=-0.0051, tangentMode=RCTM_Break, time=0, value=1}
          {arriveTangent=-0.247, interpMode=RCIM_Cubic, tangentMode=RCTM_Break, time=1, value=0.9}
         ]}}})
    )
    ApplyForce=false
    ApplyPosition=true
    ApplyPositionFalloff=false
   )
   &ParticleState()
   &SolveForcesAndVelocity(Clamp Velocity=false, Limit Acceleration=false)
   &DynamicMaterialParameters(
    Index 0 Param 2=&MakeCustomFloatFromBool(Bool=$User.bIsPainted, True Float=1, False Float=1)
    Index 0 Param 3=&MakeCustomFloatFromBool(Bool=$User.bIsPainted, True Float=1, False Float=1)
    Index 0 Param 4=0.35
    Module.Index0Param0Write=true
    Module.Index0Param1Write=true
    Module.Index0Param2Write=true
    Module.Index0Param3Write=true
   )
   &ScaleRibbonWidth(Ribbon Width Scale=2) {module=ScaleRibbonWidth001}
   &VC_UpdateRibbon_v1(Enable Skeletal Position=true)
   /SetVariables(
    Particles.R1Dyna0RedStrobes:Vector4f=&MakeVector4(X=2, Y=1, Z=&MakeCustomFloatFromBool(Bool=$User.bIsPainted, True Float=1, False Float=2), W=0.4)
    Particles.R1ScaleColorRedStrobes:LinearColor=&VC_DynamicScaleColor(
     Color to Scale:LinearColor=$Particles.R1 Color
     RGB Scale:Vector3f=<1 1 1>
     Alpha Scale:NiagaraFloat=&FloatFromCurve(
      FloatCurve=DataInterface() {propertyValues {Curve {keys [
          {arriveTangent=8.5748, interpMode=RCIM_Cubic, leaveTangent=8.5748, tangentMode=RCTM_User, time=0, value=2.25}
          {arriveTangent=1.9737, interpMode=RCIM_Cubic, leaveTangent=1.9737, tangentMode=RCTM_User, time=0.1, value=3}
          {arriveTangent=-7.3322, interpMode=RCIM_Cubic, leaveTangent=-7.3322, tangentMode=RCTM_User, time=0.7, value=1}
          {arriveTangent=-4.5533, interpMode=RCIM_Cubic, leaveTangent=-4.5533, tangentMode=RCTM_User, time=1, value=-1.5}
         ]}}}
      Scale Curve=0.5
     )
    )
    Particles.R2Dyna0:Vector4f=<4 1 1 0.5>
    Particles.R3Dyna0:Vector4f=<6 1 1 0.25>
    Particles.R4Dyna0:Vector4f=<8 1 1 0.6>
    Particles.R5RWScale:NiagaraFloat=&Multiply_Float(A=60, B=3)
    Particles.R5ScaleColorBloom:LinearColor=&VC_DynamicScaleColor(
     Color to Scale:LinearColor=$Particles.R2 Color
     RGB Scale:Vector3f=<1 1 1>
     Alpha Scale:NiagaraFloat=&FloatFromCurve(
      FloatCurve=DataInterface() {propertyValues {Curve {keys [
          {interpMode=RCIM_Cubic, leaveTangent=5.9517, tangentMode=RCTM_Break, time=0, value=0.4}
          {interpMode=RCIM_Cubic, tangentMode=RCTM_Break, time=0.1, value=1.05}
          {arriveTangent=-2.665, interpMode=RCIM_Cubic, tangentMode=RCTM_Break, time=0.8, value=0}
         ]}}}
      Scale Curve=3
     )
    )
   ) {isSetParameters=true}
   /SetVariables(
    Particles.R0Position:NiagaraPosition=&AddVectorToPosition(Position=$Particles.Position, Vector=/VehicleCosmetics/FX/Niagara/Modules/VC_ConvertVectorToWorld(Vector:Vector3f=<0 10 -30>))
    Particles.R1Position:NiagaraPosition=&AddVectorToPosition(Position=$Particles.Position, Vector=&VC_ConvertVectorToWorld(Vector:Vector3f=<0 0 -12>))
    Particles.R2Position:NiagaraPosition=&AddVectorToPosition(Position=$Particles.Position, Vector=&VC_ConvertVectorToWorld(Vector:Vector3f=<0 -10 -30>))
    Particles.R3Position:NiagaraPosition=&AddVectorToPosition(Position=$Particles.Position, Vector=&VC_ConvertVectorToWorld(Vector:Vector3f=<0 10 2>))
    Particles.R4Position:NiagaraPosition=&AddVectorToPosition(Position=$Particles.Position, Vector=&VC_ConvertVectorToWorld(Vector:Vector3f=<0 -10 2>))
    Particles.R5Position:NiagaraPosition=&AddVectorToPosition(Position=$Particles.Position, Vector=&VC_ConvertVectorToWorld(Vector:Vector3f=<0 0 -12>))
   ) {isSetParameters=true}
  ]
  rendererClasses [&NiagaraRibbonRendererProperties]
  renderers [
   {rendererClass=&NiagaraRibbonRendererProperties}
   {rendererClass=&NiagaraRibbonRendererProperties, rendererIndex=1}
   {rendererClass=&NiagaraRibbonRendererProperties, rendererIndex=2}
   {rendererClass=&NiagaraRibbonRendererProperties, rendererIndex=3}
   {rendererClass=&NiagaraRibbonRendererProperties, rendererIndex=4}
   {rendererClass=&NiagaraRibbonRendererProperties, rendererIndex=5}
   {rendererClass=&NiagaraRibbonRendererProperties, rendererIndex=6}
   {rendererClass=&NiagaraRibbonRendererProperties, rendererIndex=7}
   {rendererClass=&NiagaraRibbonRendererProperties, rendererIndex=8}
   {rendererClass=&NiagaraRibbonRendererProperties, rendererIndex=9}
   {rendererClass=&NiagaraRibbonRendererProperties, rendererIndex=10}
  ]
 }
 WheelWrap001 {
  emitterUpdateScript [
   &EmitterState(Life Cycle Mode=::NewEnumerator1, Loop Behavior=::NewEnumerator1, Loop Duration Mode=::NewEnumerator1, UseLoopDelay=false)
   &SpawnBurst_Instantaneous(Use Spawn Probability=false, UseLoopCountLimit=false) {module=SpawnBurst_Instantaneous001}
   &SpawnBurst_Instantaneous(Use Spawn Probability=false, UseLoopCountLimit=false)
   /SetVariables(Emitter.Mask_wheel:NiagaraFloat=&Sine(Normalized Angle=$Emitter.Age, Period=1, Scale=1, Bias=3)) {isSetParameters=true}
  ]
  particleSpawnScript [
   &VC_PaintedVariantsOverride(
    Black=<rgba 0 0 0 1>
    BurntSienna=<rgba 0.3438 0.0749 0 0.5>
    Cobalt=<rgba 0.0339 0.0829 0.25 0.25>
    Crimson=<rgba 0.32 0.0308 0.0592 1>
    ForestGreen=<rgba 0.0395 0.4271 0.0021 0.05>
    Gold:LinearColor=<rgba 0.8 0.5546 0.0958 0.03>
    Grey=<rgba 0.8 0.5546 0.0958 0.03>
    Lime=<rgba 0.0498 0.35 0.0164 0.2>
    Orange=<rgba 0.4 0.0954 0.0396 0.2>
    Pink=<rgba 0.4 0.02 0.3367 0.2>
    Purple=<rgba 0.1939 0.0218 0.3229 0.2>
    Saffron=<rgba 0.25 0.2176 0 0.2>
    SkyBlue=<rgba 0 0.2807 0.3 0.2>
    TitaniumWhite=<rgba 0.25 0.25 0.25 1>
   )
   &InitializeParticle(
    Color=&MakeCustomLinearColorFromBool(Bool=$User.bIsPainted, True Color=$Particles.Color, False Color=<rgba 0.1094 0.1574 0.35 0.3>)
    Lifetime=0.1
    Mesh Scale Mode=::NewEnumerator4
    Sprite Rotation Mode=::NewEnumerator2
    Sprite Size Mode=::NewEnumerator3
    UsePositionOffset=false
    Write Material Random=true
   )
   &PartitionParticles()
   &SkeletalMeshLocation(
    Bone / Socket Index=&Add_Integer(A=$Particles.PartitionParticles.Partition, B=4)
    Bone / Socket Sampling Mode=::NewEnumerator5
    Interpolate Mesh Sampling=false
    MaskSpawnGroup=false
    Mesh Sampling Type=::NewEnumerator6
    Skeletal Mesh=$System.SkeletalMesh
   )
  ]
  particleUpdateScript [
   &FN_DisableTireFX(
    Mask Index Attribute:NiagaraInt32=$Particles.PartitionParticles.Partition
    Mask Index FL=2
    Mask Index FR=3
    Mask Index RL=0
    Mask Index RR=1
   )
   &VC_WheelWrapScale(Wheel Index Attribute:NiagaraInt32=&Add_Integer(A=$Particles.PartitionParticles.Partition, B=2))
   &ScaleColor(Scale RGB=&Multiply_VectorByFloat(Vector=<1 1 1>, Float=0.5), ScaleA=true, ScaleRGB=true)
   &DynamicMaterialParameters(
    Index 0 Param 1=10
    Index 0 Param 2=1
    Index 0 Param 3=3
    Index 0 Param 4=0.6
    Module.Index0Param0Write=true
    Module.Index0Param1Write=true
    Module.Index0Param2Write=true
    Module.Index0Param3Write=true
   ) {enabled=false}
   &ParticleState(Kill Particles When Lifetime Has Elapsed=false)
   &UpdateMeshOrientation(Rotation Coordinate Space=::NewEnumerator2, Rotation Vector=<0 1 0>) {enabled=false}
   &SkeletalMeshLocation(
    Bone / Socket Index=&Add_Integer(A=$Particles.PartitionParticles.Partition, B=4)
    Bone / Socket Sampling Mode=::NewEnumerator5
    MaskSpawnGroup=false
    Mesh Sampling Type=::NewEnumerator6
    Skeletal Mesh=$System.SkeletalMesh
   ) {module=SkeletalMeshLocation001}
   &SolveForcesAndVelocity(Clamp Velocity=false, Limit Acceleration=false)
  ]
  rendererClasses [&NiagaraMeshRendererProperties]
  renderers [{rendererClass=&NiagaraMeshRendererProperties, rendererIndex=0}]
 }
}
summary {
 systemName=NS_Trail_Strobe_MP
 emitters [
  {emitterName=WheelWrap001, rendererClasses [&NiagaraMeshRendererProperties]}
  {emitterName=Ribbons002, rendererClasses [&NiagaraRibbonRendererProperties]}
 ]
 userVariables [
  User.bIsFrontendPreview {type=NiagaraBool}
  User.bIsPainted {type=NiagaraBool}
  User.Color_ScaleCurve {type=NiagaraFloat, defaultValue=0.5, description=""}
  User.ColorVariantID {type=NiagaraInt32}
  User.offset {type=Vector3f, description="offset from wheels(set through cosmetic template)", defaultValue {struct=Vector3f, value=<0 0 0>}}
  User.PaintedColor {type=LinearColor, description="", defaultValue {struct=LinearColor, value=<rgba 1 1 1 1>}}
  User.PaintedColor_2 {type=LinearColor, description="", defaultValue {struct=LinearColor, value=<rgba 1 1 1 1>}}
  User.PaintedColor_3 {type=LinearColor, description="", defaultValue {struct=LinearColor, value=<rgba 1 1 1 1>}}
  User.PaintedColor_4 {type=LinearColor, description="", defaultValue {struct=LinearColor, value=<rgba 1 1 1 1>}}
  User.PaintedScalar {type=NiagaraFloat, defaultValue=1, description=""}
  User.VehicleFlags {type=NiagaraInt32, defaultValue=806289408}
  User.Wheel_Radius_Array {
   type=NiagaraDataInterfaceArrayFloat
   defaultValue {
    struct=NiagaraExt_VariableValue_DataInterface
    value {
     dataInterface=&self:NiagaraDataInterfaceArrayFloat_0
     dataInterfaceClass=&NiagaraDataInterfaceArrayFloat
    }
   }
  }
  User.Wheel_Width_Array {
   type=NiagaraDataInterfaceArrayFloat
   defaultValue {
    struct=NiagaraExt_VariableValue_DataInterface
    value {
     dataInterface=&self:NiagaraDataInterfaceArrayFloat_1
     dataInterfaceClass=&NiagaraDataInterfaceArrayFloat
    }
   }
  }
  User.WheelSize {type=NiagaraFloat, defaultValue=0, description=""}
 ]
}
